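# Niagara idle effect for the bright Icicle Cane backpack, emitting drifting, tinted dust sprites around its sockets.
# One packed emitter runs three spawn groups with separate motion, spawning across four socket partitions at array-picked offsets.
# Extra color and size attribute sets accompany four sprite renderers.
# Variant, time of day and front-end context adjust the look; owner speed fades alpha, VelocityFade accelerates aging, and HideFX suppresses spawning and kills particles.
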NiagaraSystem /BRCosmetics/Accessories/FORT_Backpacks/Backpack_F_MED_IcicleCane_Bright/FX/NS_Backpack_IcicleCane_Bright_Idle
dependencies {
 usedDataInterfaces [
  &NiagaraDataInterfaceCurve
  &NiagaraDataInterfaceArrayFloat3
  &NiagaraDataInterfaceColorCurve
  &NiagaraDataInterfaceSpriteRendererInfo
  &NiagaraDataInterfaceSocketReader
 ]
 usedDynamicInputs [
  &MergeHelper_Vector2Switch
  /Game/Effects/Niagara/DynamicInputs/Math/FN_InterpFloat
  &Divide_Float
 ]
 usedModules [
  &self:SystemSpawnScript.SystemScriptSource.SystemScriptGraph.NiagaraNodeAssignment_9.SetVariables_#1
  &self:SystemSpawnScript.SystemScriptSource.SystemScriptGraph.NiagaraNodeAssignment_0.SetVariables_#2
  &SystemState
  &FN_VariantManager
  &self:SystemSpawnScript.SystemScriptSource.SystemScriptGraph.NiagaraNodeAssignment_2.SetVariables_#3
  &self:SystemSpawnScript.SystemScriptSource.SystemScriptGraph.NiagaraNodeAssignment_1.SetVariables_#4
  &self:SystemSpawnScript.SystemScriptSource.SystemScriptGraph.NiagaraNodeAssignment_3.SetVariables_#5
 ]
}
emitters {Dust_Packed {
  emitterUpdateScript [
   &EmitterState(Inactive Response=::NewEnumerator1, Life Cycle Mode=::NewEnumerator1, Loop Duration=1, UseLoopDelay=false)
   &SpawnRate(
    Spawn Probability=&ClampFloat(Float=&OneMinusFloat(Float=$User.HideFX))
    SpawnRate=25
    Use Spawn Probability=true
   )
   &SpawnRate(
    Spawn Probability=&ClampFloat(Float=&OneMinusFloat(Float=$User.HideFX))
    SpawnGroup=1
    SpawnRate=15
    Use Spawn Probability=true
   ) {module=SpawnRate001}
   &SpawnRate(
    Spawn Probability=&ClampFloat(Float=&OneMinusFloat(Float=$User.HideFX))
    SpawnGroup=2
    SpawnRate=5
    Use Spawn Probability=true
   ) {module=SpawnRate002}
  ]
  particleSpawnScript [
   /SetVariables(Particles.PerParticleRandom:NiagaraFloat=&RandomRangeFloat(Minimum=0, Maximum=1)) {isSetParameters=true}
   &PartitionParticles(Number Of Partitions=4, Partition Mode:ENiagara_PartitionMode_v2=::NewEnumerator5)
   &InitializeParticle(
    Color=&VariantColorSelector(
     Variant Amount=0
     Original Color=&RandomRangeLinearColor(Minimum=<rgba 0.2563 0.299 0.4271 0.6>, Maximum=<rgba 0.6 0.6667 1 0.8>)
     Variant 1 Color=&RandomRangeLinearColor(Minimum=<rgba 0.4 0.6733 0.8 0.4>, Maximum=<rgba 0.5 0.7083 1 0.6>)
     Variant 2 Color=?
    )
    Lifetime=&MergeHelper_FloatSwitch(
     0=&RandomRangeFloat(Minimum=0.85, Maximum=1.4)
     1=&RandomRangeFloat(Minimum=0.6, Maximum=1)
     2=&RandomRangeFloat(Minimum=0.6, Maximum=1.4)
     3=?
     Index=$Emitter.SpawnGroup
     Num=::NewEnumerator1
    )
    Sprite Rotation Mode=::NewEnumerator2
    Sprite Size Mode=::NewEnumerator3
    Sprite UV Mode=::NewEnumerator3
    Uniform Sprite Size=&RandomRangeFloat(Minimum=20, Maximum=28)
    UsePositionOffset=false
    Write Material Random=true
   )
   &CameraOffset(Camera Offset Amount=&FN_FrontEndFloatModifier(
     Gameplay=&MergeHelper_FloatSwitch(0=10, 1=15, 2=10, 3=?, Index=$Emitter.SpawnGroup, Num=::NewEnumerator1)
     Front End=7
    ))
   &SocketLocation(
    SamplePosOffset=true
    Sampled Position Offset=&MergeHelper_VectorSwitch(
     0=&SelectVectorFromArray(
      Array Sampling Mode=::NewEnumerator0
      Direct Array Index=?
      Vector Selection Array=DataInterface() {propertyValues {FloatData [
         <-5.3996 -0.522 1.8644>
         <-7.0154 1.352 -1.318>
         <-6.2677 0.7967 -8.5158>
         <-1.6017 0.9576 3.0518>
         <1.1194 -0.1139 -6.7669>
         <1.1248 -0.1514 -5.6194>
         <2.3359 -0.1341 -6.2445>
         <4.1516 -0.1081 -0.125>
         <5.6186 -0.4103 -0.5708>
         <8.2088 -0.2557 -1.8934>
         <6.7854 -0.4496 -1.9105>
        ]}}
      Random Seed=0
     )
     1=&SelectVectorFromArray(
      Array Sampling Mode=::NewEnumerator0
      Direct Array Index=?
      Vector Selection Array=DataInterface() {propertyValues {FloatData [
         <-5.4579 1.0661 -1.6854>
         <-10.375 1.1057 6.9624>
         <-0.9874 -0.2962 6.052>
         <3.4083 -0.0416 4.8012>
         <13.2174 0.1741 0.3236>
         <-2.6564 0.0434 -4.936>
         <2.7811 0.264 -1.6921>
         <-1.6622 1.056 -1.2027>
         <-6.5254 0.9031 7.6335>
         <-1.7209 0.6108 7.7948>
         <-8.7281 -0.6066 -4.3311>
        ]}}
      Random Seed=0
     )
     2=&SelectVectorFromArray(
      Array Sampling Mode=::NewEnumerator0
      Direct Array Index=?
      Vector Selection Array=DataInterface() {propertyValues {FloatData [
         <-3.5441 -0.7231 -1.8973>
         <1.4968 -0.3532 -0.8912>
         <-0.2309 -0.4722 3.0212>
         <-2.9939 0.4103 -5.9922>
         <7.1373 0.5126 -2.5061>
         <10.7467 0.2124 -6.8602>
         <4.6209 0.7078 -1.1097>
        ]}}
      Random Seed=0
     )
     3=&SelectVectorFromArray(
      Array Sampling Mode=::NewEnumerator0
      Direct Array Index=?
      Vector Selection Array=DataInterface() {propertyValues {FloatData [<-2.0301 -0.2223 1.2521>, <2.3524 0.1004 -0.248>, <-4.5804 -0.2746 -5.8121>, <7.3034 -0.0999 5.6296>]}}
      Random Seed=0
     )
     Index=$Particles.PartitionParticles.Partition
     Num=::NewEnumerator2
    )
    Socket Index=$Particles.PartitionParticles.Partition
    Socket Reader=$System.Socket Reader
    Socket Sampling Mode=::NewEnumerator3
    VelocitySpeedLimit=false
   )
   &AddVelocity(
    Module.Use Falloff Distance=false
    Origin Offset=&RotateVectorByQuaternion(Vector To Rotate=<0 1 -4>, Quaternion=$Output.SocketLocation.SampledRotation)
    Override Seed=false
    Velocity Mode=::NewEnumerator1
    Velocity Speed=&MergeHelper_FloatSwitch(
     0=&RandomRangeFloat(Minimum=0, Maximum=7)
     1=&RandomRangeFloat(Minimum=8, Maximum=14)
     2=&RandomRangeFloat(Minimum=4, Maximum=10)
     3=?
     Index=$Emitter.SpawnGroup
     Num=::NewEnumerator1
    )
   )
   /SetVariables(Particles.RotationDirection:NiagaraFloat=&UniformAOrBFloat(A=-1, Evaluation Type=::NewEnumerator0, Recalculate Random Each Loop=true, Override Seed=false)) {isSetParameters=true}
   /SetVariables(Particles.VisibilityTag=$Emitter.SpawnGroup) {isSetParameters=true}
  ]
  particleUpdateScript [
   &ParticleState()
   &SpriteRotationRate(Rotation Rate=&Multiply_Float(
     A=$Particles.RotationDirection
     B=&MergeHelper_FloatSwitch(
      0=0
      1=0
      2=&RandomRangeFloat(Minimum=35, Maximum=60)
      3=?
      Index=$Particles.VisibilityTag
      Num=::NewEnumerator1
     )
    ))
   &Drag(
    Drag=&FloatFromCurve(
     FloatCurve=DataInterface() {propertyValues {Curve {keys [
         {arriveTangent=0.1577, interpMode=RCIM_Cubic, leaveTangent=0.1577, tangentMode=RCTM_User, time=0, value=0.125}
         {arriveTangent=1.6704, interpMode=RCIM_Cubic, leaveTangent=1.6704, tangentMode=RCTM_User, time=1, value=1}
        ]}}}
     Scale Curve=&RandomRangeFloat(Minimum=2.5, Maximum=4)
    )
    Use Linear Drag=true
    Use Rotational Drag=true
   )
   &CurlNoiseForce(
    Bias Noise=false
    Noise Strength=&MergeHelper_FloatSwitch(
     0=0
     1=&FloatFromCurve(
      FloatCurve=DataInterface() {propertyValues {Curve {keys [{interpMode=RCIM_Cubic, tangentMode=RCTM_User, time=0, value=0.2}, {interpMode=RCIM_Cubic, tangentMode=RCTM_User, time=1, value=1}]}}}
      Scale Curve=100
     )
     2=&RandomRangeFloat(Minimum=30, Maximum=55)
     3=?
     Index=$Particles.VisibilityTag
     Num=::NewEnumerator1
    )
    Pan Noise=false
    Random Seed=33
   ) {module=CurlNoiseForce001}
   &ScaleColor(
    Scale Alpha=&FloatFromCurve(
     FloatCurve=DataInterface() {propertyValues {Curve {keys [
         {arriveTangent=7.9556, interpMode=RCIM_Cubic, leaveTangent=7.9556, tangentMode=RCTM_User, time=0, value=0}
         {arriveTangentWeight=0.0667, interpMode=RCIM_Cubic, leaveTangentWeight=0.1, tangentMode=RCTM_User, time=0.2, value=1}
         {arriveTangentWeight=0.1601, interpMode=RCIM_Cubic, leaveTangentWeight=0.2668, tangentMode=RCTM_User, time=0.5, value=1}
         {arriveTangent=-3.5681, tangentMode=RCTM_User, time=1, value=0}
        ]}}}
     Scale Curve=&Multiply_Float(
      A=&FN_FrontEndFloatModifier(Gameplay=0.13, Front End=0.06)
      B=&FNTimeOfDayScalar(Night Scale=1.2, Morning Scale=0.8, Evening Scale=0.85)
     )
    )
    Scale RGB=&MakeVectorFromLinearColorRGB(Color=&ColorFromCurve(
      ColorCurve=DataInterface() {propertyValues {
        AlphaCurve {keys [<at 0 1 RCIM_Cubic>]}
        BlueCurve {keys [<at 0.0932 1 RCIM_Cubic>,<at 0.125 1>,<at 0.5 1>,<at 0.875 1 RCIM_Cubic>,<at 0.905 1>]}
        GreenCurve {keys [
          <at 0.0932 0.5833 RCIM_Cubic>
          <at 0.125 1>
          <at 0.5 1>
          {arriveTangent=-1.0288, interpMode=RCIM_Cubic, leaveTangent=-1.0288, time=0.875, value=1}
          <at 0.905 0.5833>
         ]}
        RedCurve {keys [
          <at 0.0932 0.5 RCIM_Cubic>
          <at 0.125 1>
          <at 0.5 1>
          {arriveTangent=-1.2346, interpMode=RCIM_Cubic, leaveTangent=-1.2346, time=0.875, value=1}
          <at 0.905 0.5>
         ]}
       }}
      CurveIndex=Linked() {linkedVariable {name=NPC.FortniteNPC.RemappedTimeOfDay_0-1, type=NiagaraFloat}}
     ))
    ScaleA=true
    ScaleRGB=true
   )
   &ScaleSpriteSize(
    Scale Sprite Size Mode=::NewEnumerator0
    Uniform Scale Factor=&FloatFromCurve(FloatCurve=DataInterface() {propertyValues {Curve {keys [
         {arriveTangent=0.3971, interpMode=RCIM_Cubic, leaveTangent=0.3971, tangentMode=RCTM_User, time=0, value=0.8}
         {arriveTangent=0.2599, interpMode=RCIM_Cubic, leaveTangent=0.2599, tangentMode=RCTM_User, time=0.3, value=0.9}
         {interpMode=RCIM_Cubic, leaveTangent=0.1633, time=1, value=1}
        ]}}})
   )
   &SubUVAnimation(
    Output Alternate Renderer Binding=true
    "Sprite Renderer "=DataInterface() {propertyValues {SpriteRenderer=&self:Dust_Packed_0.NiagaraSpriteRendererProperties_2}}
    UseEndFrame=false
    UseStartFrame=false
   )
   &FN_Sprite_Attributes(
    Color=&MultiplyLinearColor(
     Linear Color=&ColorFromCurve(ColorCurve=DataInterface() {propertyValues {
        AlphaCurve {keys [<at 0 0>,<at 0.3225 0.2 RCIM_Cubic>,<at 0.9968 0>]}
        BlueCurve {keys [<at 0.0028 5>, {arriveTangent=-4.2116, interpMode=RCIM_Cubic, leaveTangent=-4.2116, time=0.2108, value=3}, <at 1 0.8 RCIM_Cubic>]}
        GreenCurve {keys [<at 0.0028 4.401>, {arriveTangent=-3.6569, interpMode=RCIM_Cubic, leaveTangent=-3.6569, time=0.2108, value=2.7927}, <at 1 0.7542 RCIM_Cubic>]}
        RedCurve {keys [
          <at 0.0028 3.8021>
          {arriveTangent=-3.1023, interpMode=RCIM_Cubic, leaveTangent=-3.1023, time=0.2108, value=2.2969}
          <at 1 0.7083 RCIM_Cubic>
         ]}
       }})
     Scale Factor=&MakeVector4FromVectorAndFloat(
      "Vector (RGB)"=&MakeVectorFromLinearColorRGB(Color=&VariantColorSelector(
        Variant Amount=0
        Original Color=&RandomRangeLinearColor(Minimum=<rgba 0.6 0.9 1 1>, Maximum=<rgba 0.515 0.6767 1 1>, Color Channel Mode=::NewEnumerator2)
        Variant 1 Color=&RandomRangeLinearColor(Minimum=<rgba 0.2 0.7533 1 1>, Maximum=<rgba 0.165 0.5129 1 1>, Color Channel Mode=::NewEnumerator2)
        Variant 2 Color=?
       ))
      "Float (Alpha)"=&Multiply_Float(
       A=$System.TODScalar
       B=&Lerp_Float(
        A=0
        B=1
        Alpha=&RemapRange(
         Input Value To Remap=&VectorLength(Vector=$Engine.Owner.Velocity)
         Input Min=0
         Input Max=80
         Output Min=1
         Output Max=0
        )
       )
      )
     )
    )
    Create Color Attribute=true
    Create SpriteSize Attribute=true
    SpriteSize=&Multiply_Vector2D_ByFloat(
     Float=&FloatFromCurve(
      FloatCurve=DataInterface() {propertyValues {Curve {keys [
          {tangentMode=RCTM_User, time=0, value=1}
          {arriveTangent=-0.0405, interpMode=RCIM_Cubic, leaveTangent=-0.0405, tangentMode=RCTM_User, time=0.7009, value=1}
          {arriveTangent=-8.0089, interpMode=RCIM_Cubic, leaveTangent=-8.0089, tangentMode=RCTM_User, time=1, value=0}
         ]}}}
      Scale Curve=&RandomRangeFloat(Minimum=1, Maximum=&FN_FrontEndFloatModifier(Gameplay=2.5, Front End=1.9))
     )
     Vector2D=&Multiply_Vector2D(
      A=<0.535 0.535>
      B=&MakeVector2D(
       X=&FloatFromCurve(
        FloatCurve=DataInterface() {propertyValues {Curve {keys [<at 0 1>,<at 0.8 2 RCIM_Constant>,<at 0.95 3 RCIM_Constant>]}}}
        CurveIndex=$Particles.PerParticleRandom
       )
       Y=&FloatFromCurve(
        FloatCurve=DataInterface() {propertyValues {Curve {keys [<at 0 1>,<at 0.8 2 RCIM_Constant>,<at 0.95 3>]}}}
        CurveIndex=$Particles.PerParticleRandom
       )
      )
     )
    )
    Write to Module Namespace=true
   )
   &FN_Sprite_Attributes(
    Color=&MultiplyLinearColor(
     Linear Color=$Particles.FN_Sprite_Attributes.Color
     Scale Factor=&MakeVector4FromVectorAndFloat("Vector (RGB)"=&MakeVectorFromLinearColorRGB(Color=<rgba 3 3 3 1>), "Float (Alpha)"=0.15)
    )
    Create Color Attribute=true
    Create SpriteSize Attribute=true
    SpriteSize=&Multiply_Vector2D_ByFloat(Float=1.5, Vector2D=$Particles.FN_Sprite_Attributes.SpriteSize)
    Write to Module Namespace=true
   ) {module=FN_Sprite_Attributes001}
   &FN_Sprite_Attributes(
    Color=&MultiplyLinearColor(
     Linear Color=&VariantColorSelector(
      Variant Amount=0
      Original Color=&RandomRangeLinearColor(Minimum=<rgba 1.575 2.6437 3 1>, Maximum=<rgba 1.5 1.625 3 1>)
      Variant 1 Color=&RandomRangeLinearColor(Minimum=<rgba 0.375 2.125 3 1>, Maximum=<rgba 0.3 1.515 3 1>)
      Variant 2 Color=?
     )
     Scale Factor=&MakeVector4FromVectorAndFloat("Vector (RGB)"=<1 1 1>, "Float (Alpha)"=$System.TODScalar)
    )
    Create Color Attribute=true
    Create SpriteSize Attribute=true
    SpriteSize=&Multiply_Vector2D_ByFloat(
     Float=&FloatFromCurve(FloatCurve=DataInterface() {propertyValues {Curve {
         defaultValue=1
         keys [
          {interpMode=RCIM_Cubic, tangentMode=RCTM_User, time=0, value=0}
          {arriveTangent=-0.1, arriveTangentWeight=0.075, interpMode=RCIM_Cubic, leaveTangent=-0.1, tangentMode=RCTM_User, time=0.225, value=1}
          {arriveTangent=-1.8257, arriveTangentWeight=0.2583, interpMode=RCIM_Cubic, leaveTangent=-1.8257, tangentMode=RCTM_User, time=1, value=0}
         ]
        }}})
     Vector2D=&Vector2DFromFloat(Value=&RandomRangeFloat(Minimum=3, Maximum=6))
    )
    Write to Module Namespace=true
   ) {module=FN_Sprite_Attributes002}
   /SetVariables(Particles.Age:NiagaraFloat=&Add_Float(
     A=$Particles.Age
     B=&Multiply_Float(A=$System.VelocityFade, B=&Multiply_Float(A=$Particles.Lifetime, B=$Engine.DeltaTime))
    )) {isSetParameters=true}
   &GravityForce(Gravity=&MakeVector(
     X=0
     Y=0
     Z=&MergeHelper_FloatSwitch(
      0=&RandomRangeFloat(Minimum=-3, Maximum=-15)
      1=&RandomRangeFloat(Minimum=0, Maximum=-25)
      2=&RandomRangeFloat(Minimum=0, Maximum=-25)
      3=?
      Index=$Particles.VisibilityTag
      Num=::NewEnumerator1
     )
    ))
   &SolveForcesAndVelocity(Clamp Velocity=false, Limit Acceleration=false)
   &KillParticles(Kill Particles=&SetBoolByFloatComparison(A=$User.HideFX, B=1, Comparison Type=::NewEnumerator4))
  ]
  rendererClasses [&NiagaraSpriteRendererProperties]
  renderers [
   {rendererClass=&NiagaraSpriteRendererProperties}
   {rendererClass=&NiagaraSpriteRendererProperties, rendererIndex=1}
   {rendererClass=&NiagaraSpriteRendererProperties, rendererIndex=2}
   {rendererClass=&NiagaraSpriteRendererProperties, rendererIndex=3}
  ]
 }}
summary {
 systemName=NS_Backpack_IcicleCane_Bright_Idle
 emitters [{emitterName=Dust_Packed, rendererClasses [&NiagaraSpriteRendererProperties]}]
 userVariables [
  User.bIsFrontEnd {type=NiagaraBool}
  User.bIsFrontEndPreview {type=NiagaraBool}
  User.HideFX {type=NiagaraFloat, defaultValue=0, description=""}
  User.Variant {type=NiagaraFloat, defaultValue=0, description=""}
 ]
}
